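# Animation Blueprint for the De Las Muertos backpack cosmetic, adding secondary physics motion to its zipper bones.
# The input pose passes through an additive offset on allOffsetA, then chained Anim Dynamics nodes for dyn_topZip and dyn_mouthZip with angular limits.
# A separate three-node Anim Dynamics chain for the jaw bones has no input pose and never reaches Output Pose.
# The EventGraph holds only an empty update event.
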AnimBlueprint /BRCosmetics/Accessories/FORT_Backpacks/F_MED_DeLasMuertos_Backpack/Meshes/F_MED_DelasMuertos_Backpack_AnimBP
target_skeleton=&/F_MED_DelasMuertos_Backpack_Skeleton
graphs [
 AnimGraph {nodes [
   n1 = Anim Dynamics - Bone: dyn_JawALL(Alpha=0.2, ExternalForce="0,0,0") {
    angularDampingOverride=0.999
    angularSpringConstant=1500
    bAngularSpring=true
    bGravityOverrideInSimSpace=true
    bOverrideAngularDamping=true
    bUseGravityOverride=true
    alphaBoolBlend {blendOption=Linear}
    boundBone {boneName=dyn_JawALL}
    componentAppliedLinearAccClamp=<1000 1000 1000>
    componentLinearAccScale=<1 1 1>
    componentLinearVelScale=<1 1 3>
    gravityOverride {z=-981}
    physicsBodyDefinitions [{
      sphereCollisionRadius=0
      boundBone {boneName=dyn_JawALL}
      boxExtents=<36 36 20>
      constraintSetup {twistAxis=AxisZ, angularLimitsMax=<15 15 0>, angularLimitsMin=<-15 -15 0>, angularTarget=<0 0 0>}
      localJointOffset=<-0.1 0 -16>
     }]
   }
   n2 = Anim Dynamics - Bone: dyn_lowJaw(ComponentPose=@n1, Alpha=1, ExternalForce="0,0,0") {
    angularDampingOverride=0.99
    bGravityOverrideInSimSpace=true
    bOverrideAngularDamping=true
    alphaBoolBlend {blendOption=Linear}
    boundBone {boneName=dyn_lowJaw}
    componentAppliedLinearAccClamp=<1000 1000 1000>
    componentLinearAccScale=<1 1 1>
    componentLinearVelScale=<1 1 3>
    physicsBodyDefinitions [{
      sphereCollisionRadius=0
      boundBone {boneName=dyn_lowJaw}
      boxExtents=<14 14 6>
      constraintSetup {twistAxis=AxisZ, angularLimitsMax=<10 15 0>, angularLimitsMin=<-10 -15 0>, angularTarget=<0 0 0>}
      localJointOffset=<-0.1 0 -6>
     }]
   }
   Anim Dynamics - Bone: dyn_upJaw(ComponentPose=@n2, Alpha=1, ExternalForce="0,0,0") {
    angularDampingOverride=0.99
    bGravityOverrideInSimSpace=true
    bOverrideAngularDamping=true
    bUseGravityOverride=true
    alphaBoolBlend {blendOption=Linear}
    boundBone {boneName=dyn_upJaw}
    componentAppliedLinearAccClamp=<1000 1000 1000>
    componentLinearAccScale=<1 1 1>
    componentLinearVelScale=<1 1 3>
    gravityOverride {z=-981}
    physicsBodyDefinitions [{
      sphereCollisionRadius=0
      boundBone {boneName=dyn_upJaw}
      boxExtents=<9 9 4>
      constraintSetup {twistAxis=AxisZ, angularLimitsMax=<10 15 0>, angularLimitsMin=<-10 -15 0>, angularTarget=<0 0 0>}
      localJointOffset=<-0.1 0 -5>
     }]
   }
   AnimGraphNode_SubInput_0 = Input Pose()
   n3 = "Transform (Modify) Bone - Bone: allOffsetA"!(ComponentPose=Local To Component(LocalPose=@AnimGraphNode_SubInput_0), Alpha=1, Translation="0,0,2.5", Rotation="0,0,5") -> (Pose) {
    rotationMode=BMM_Additive
    translationMode=BMM_Additive
    alphaBoolBlend {blendOption=Linear}
    boneToModify {boneName=allOffsetA}
    rotation {roll=5}
    translation {z=2.5}
   }
   n4 = Anim Dynamics - Bone: dyn_topZip(ComponentPose=@n3, Alpha=1, ExternalForce="0,0,0") {
    angularDampingOverride=0.995
    bGravityOverrideInSimSpace=true
    bOverrideAngularDamping=true
    bUseGravityOverride=true
    alphaBoolBlend {blendOption=Linear}
    boundBone {boneName=dyn_topZip}
    componentAppliedLinearAccClamp=<1000 1000 1000>
    componentLinearAccScale=<1 1 1>
    componentLinearVelScale=<1 1 3>
    gravityOverride {z=-921}
    physicsBodyDefinitions [{
      sphereCollisionRadius=0
      boundBone {boneName=dyn_topZip}
      boxExtents=<8 8 4>
      constraintSetup {twistAxis=AxisZ, angularLimitsMax=<1 25 0>, angularLimitsMin=<-1 -3 0>, angularTarget=<0 0 0>}
      localJointOffset=<0 0 -3>
     }]
   }
   n5 = Anim Dynamics - Bone: dyn_mouthZip(ComponentPose=@n4, Alpha=1, ExternalForce="0,0,0") {
    angularDampingOverride=0.995
    bGravityOverrideInSimSpace=true
    bOverrideAngularDamping=true
    alphaBoolBlend {blendOption=Linear}
    boundBone {boneName=dyn_mouthZip}
    componentAppliedLinearAccClamp=<1000 1000 1000>
    componentLinearAccScale=<1 1 1>
    componentLinearVelScale=<1 1 3>
    physicsBodyDefinitions [{
      sphereCollisionRadius=0
      boundBone {boneName=dyn_mouthZip}
      boxExtents=<4 4 2>
      constraintSetup {twistAxis=AxisZ, angularLimitsMax=<20 1 0>, angularLimitsMin=<-10 -1 0>, angularTarget=<0 0 0>}
      localJointOffset=<0 0 -3>
     }]
   }
   Output Pose(Result=Component To Local(ComponentPose=@n5)) {layerGroup=None}
  ]}
 EventGraph {nodes [Event BlueprintUpdateAnimation()]}
]
variables []
